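# Physics asset for the Manual vehicle body mesh, defining the simple collision used for that cosmetic body.
# It has a single body on the root bone made of one box that roughly encloses the vehicle, offset slightly forward and raised above the origin.
# It has no constraints, so the body acts as one rigid shape with no articulated parts.
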PhysicsAsset /VehicleCosmetics/Bodies/Manual/Meshes/Body_Manual_PhysicsDM
bodies {root [{shapeType=Box, extentX=444, extentY=216, extentZ=116, shapeName=root_box, center=<28 0 101>}]}
constraints []
